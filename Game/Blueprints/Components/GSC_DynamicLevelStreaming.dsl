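; GSC_DynamicLevelStreaming is a component Blueprint that gives audio feedback when nav actors are hovered or selected.
; Each of its two events plays a fixed 2D Winterfest UI cue, one for hover and one for press.
; Despite the name, the graph shown contains no level streaming logic, only these sound calls.
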(event EventOnNavActorHovered
  (Audio|PlaySound2D "/Game/Athena/Sounds/Events/Winterfest/Winterfest_UI_Hover_CUE.Winterfest_UI_Hover_Cue"))

(event EventOnNavActorSelected
  (Audio|PlaySound2D "/Game/Athena/Sounds/Events/Winterfest/Winterfest_UI_Pressed_Cue.Winterfest_UI_Pressed_Cue"))
